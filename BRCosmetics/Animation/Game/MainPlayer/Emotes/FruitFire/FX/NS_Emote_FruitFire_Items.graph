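# Niagara system for the FruitFire emote's thrown items, bursting mesh particles from a socket in two staggered spawn groups.
# Spawn group selects each particle's mesh, color and narrow launch direction.
# Gravity, a brief curve-driven drag pulse and bouncing collision move the items, which shrink away at end of life.
# Collision switches on after early flight, with platform quality masks gating two variants, one using an analytical plane below the local origin.
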NiagaraSystem /BRCosmetics/Animation/Game/MainPlayer/Emotes/FruitFire/FX/NS_Emote_FruitFire_Items
dependencies {
 usedDataInterfaces [
  &NiagaraDataInterfaceMeshRendererInfo
  &NiagaraDataInterfaceCurve
  &NiagaraDataInterfacePlatformSet
  &NiagaraDataInterfaceSocketReader
 ]
 usedDynamicInputs []
 usedModules [
  &self:SystemSpawnScript.SystemScriptSource.SystemScriptGraph.NiagaraNodeAssignment_0.SetVariables_#1
  &SystemState
 ]
}
emitters {Items {
  emitterUpdateScript [
   &EmitterState(Life Cycle Mode=::NewEnumerator1, Loop Behavior=::NewEnumerator1, Loop Duration=1, UseLoopDelay=false)
   &SpawnBurst_Instantaneous(Spawn Group=1, Spawn Time=0.1, Use Spawn Probability=false, UseLoopCountLimit=false) {module=SpawnBurst_Instantaneous001}
   &SpawnBurst_Instantaneous(Use Spawn Probability=false, UseLoopCountLimit=false)
  ]
  particleSpawnScript [
   &InitializeParticle(
    Color=&MaskLinearColorBySpawnGroup(Masked Linear Color=<rgba 1 1 1 1>, Passthrough Linear Color=<rgba 0.8698 0.4998 0.7494 1>)
    Lifetime Max=1.5
    Lifetime Min=1.3
    Lifetime Mode=::NewEnumerator1
    Mesh Renderer Array Visibility Mode=::NewEnumerator1
    Mesh Renderer Info=DataInterface() {propertyValues {MeshRenderer=&self:Items_0.NiagaraMeshRendererProperties_0}}
    Mesh Renderer Mesh Index:NiagaraInt32=&SpawnGroup()
    Mesh Scale Mode=::NewEnumerator3
    UsePositionOffset=false
    Write Material Random=true
   )
   &AddVelocity(
    Cone Angle=1
    Cone Axis=&MaskVectorBySpawnGroup(Masked Vector=<-0.3 0 1>, Passthrough Vector=<0.2 -0.3 1>)
    Module.UseConeFalloff=false
    Module.UseVelDistribution=false
    Override Seed=false
    Velocity Mode=::NewEnumerator2
    Velocity Speed=400
   )
   &SocketLocation(SamplePosOffset=false, Socket Reader=$System.Socket Reader, Socket Sampling Mode=::NewEnumerator3, VelocitySpeedLimit=false)
   &InitialMeshOrientation(Use Rotation Vector=false)
  ]
  particleUpdateScript [
   &ParticleState()
   &ScaleMeshSize(Scale Factor=&VectorFromFloat(Value=&FloatFromCurve(FloatCurve=DataInterface() {propertyValues {Curve {
         defaultValue=1
         keys [
          {tangentMode=RCTM_User, time=0, value=1}
          {arriveTangent=-0.0405, interpMode=RCIM_Cubic, leaveTangent=-0.0405, tangentMode=RCTM_User, time=0.7009, value=1}
          {arriveTangent=-8.0089, interpMode=RCIM_Cubic, leaveTangent=-8.0089, tangentMode=RCTM_User, time=1, value=0}
         ]
        }}})))
   &GravityForce()
   &Drag(
    Drag=&FloatFromCurve(
     FloatCurve=DataInterface() {propertyValues {Curve {
        defaultValue=1
        keys [
         {interpMode=RCIM_Cubic, tangentMode=RCTM_User, time=0.0372, value=0}
         {interpMode=RCIM_Cubic, tangentMode=RCTM_User, time=0.0803, value=1}
         {arriveTangentWeight=0.0114, interpMode=RCIM_Cubic, leaveTangentWeight=0.4365, tangentMode=RCTM_User, time=0.1035, value=0.1}
         <at 0.1561 0 RCIM_Cubic>
        ]
       }}}
     Scale Curve=&RandomRangeFloat(Minimum=7, Maximum=10)
    )
    Use Linear Drag=true
    Use Rotational Drag=true
   )
   &Collision(
    Age Colliding Particles=false
    Collision Enabled=&Boolean_AndOperation(
     Bool1=&IsPlatformSetActive(New Platform Set=DataInterface() {propertyValues {Platforms {qualityLevelMask=-8}}})
     Bool2=&SetBoolByFloatComparison(
      A=&FloatFromCurve(FloatCurve=DataInterface() {propertyValues {Curve {
          defaultValue=1
          keys [{interpMode=RCIM_Cubic, tangentMode=RCTM_User, time=0.2, value=0}, {interpMode=RCIM_Cubic, tangentMode=RCTM_User, time=0.3, value=1}]
         }}})
      B=0.5
     )
    )
    Control Roll On Collision=true
    Enable Rest State=false
    Friction=0.1
    KillOnCollision=false
    Module.EnableMaxCollisionCount=false
    Module.Randomize Collision Normal Vector=false
    Particle Radius Scale=5
    Radius Calculation Type=::NewEnumerator1
    Restitution=0.4
    ScaleRateOfRotation=true
   )
   &Collision(
    Age Colliding Particles=false
    Analytical Collision Plane Position 1:Vector3f=<0 0 -100>
    Analytical Collision Plane Space:ENiagaraCoordinateSpace=::Local
    CPU Collision Type=::NewEnumerator3
    Collision Enabled=&Boolean_AndOperation(
     Bool1=&IsPlatformSetActive(New Platform Set=DataInterface() {propertyValues {Platforms {qualityLevelMask=-28}}})
     Bool2=&SetBoolByFloatComparison(
      A=&FloatFromCurve(FloatCurve=DataInterface() {propertyValues {Curve {
          defaultValue=1
          keys [{interpMode=RCIM_Cubic, tangentMode=RCTM_User, time=0.2, value=0}, {interpMode=RCIM_Cubic, tangentMode=RCTM_User, time=0.3, value=1}]
         }}})
      B=0.5
     )
    )
    Control Roll On Collision=true
    Enable Rest State=false
    Friction=0.1
    KillOnCollision=false
    Module.EnableMaxCollisionCount=false
    Module.Randomize Collision Normal Vector=false
    Particle Radius Scale=5
    Perform 2 Analytical Collision Queries=false
    Radius Calculation Type=::NewEnumerator1
    Restitution=0.4
    ScaleRateOfRotation=false
   ) {module=Collision002}
   &SolveForcesAndVelocity(Clamp Velocity=false, Limit Acceleration=false, Manually Enable Rotational Solver=true)
   &ScaleMeshSizeBySpeed(Max Scale Factor=<1 1 1>)
  ]
  rendererClasses [&NiagaraMeshRendererProperties]
  renderers [{rendererClass=&NiagaraMeshRendererProperties, rendererIndex=0}]
 }}
summary {
 systemName=NS_Emote_FruitFire_Items
 emitters [{emitterName=Items, rendererClasses [&NiagaraMeshRendererProperties]}]
 userVariables [
  User.bIsFrontEnd {type=NiagaraBool}
  User.bIsFrontEndPreview {type=NiagaraBool, defaultValue=true}
  User.FrontEnd {type=NiagaraFloat, defaultValue=1, description=""}
  User.OwnerLocalVelocity {type=Vector3f, description="", defaultValue {struct=Vector3f, value=<0 0 0>}}
 ]
}
