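# Niagara system for the FruitFire emote's traversal effect, emitting short-lived toon-shaded dust puffs from sockets on the player.
# One emitter picks its spawn rate from owner speed, front-end state and platform, then partitions particles across sockets.
# A platform-dependent spawn group sets visibility tag and color (time-of-day tinted) for the mesh and sprite renderers.
# Curves drive growth, darkening, drag and material parameters, while BlendSpace keeps particle motion blended toward the moving owner.
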NiagaraSystem /BRCosmetics/Animation/Game/MainPlayer/Emotes/FruitFire/FX/NS_Emote_FruitFire_Traversal
dependencies {
 usedDataInterfaces [
  &NiagaraDataInterfacePlatformSet
  &NiagaraDataInterfaceCurve
  &NiagaraDataInterfaceColorCurve
  &NiagaraDataInterfaceArrayFloat
  &NiagaraDataInterfaceSocketReader
 ]
 usedDynamicInputs []
 usedModules [
  &self:SystemSpawnScript.SystemScriptSource.SystemScriptGraph.NiagaraNodeAssignment_0.SetVariables_#1
  &SystemState
  &BlendSpaceHelper
 ]
}
emitters {Default_Emote_TraversalDust_Toon {
  emitterSpawnScript [
   /SetVariables(
    Emitter.FE Particle Spawn Rate:NiagaraFloat=&FN_PickFloatDependOnPlatform(TargetPlatform=DataInterface() {propertyValues {Platforms {qualityLevelMask=-29}}}, TargetPlatformActive=60, TargetPlatformNoActive=30)
    Emitter.Max Traversal Speed=300
    Emitter.Particle Spawn Rate:NiagaraFloat=&FN_PickFloatDependOnPlatform(TargetPlatform=DataInterface() {propertyValues {Platforms {qualityLevelMask=-29}}}, TargetPlatformActive=60, TargetPlatformNoActive=40)
   ) {isSetParameters=true}
   /Game/Effects/Niagara/MasterMaterials/ToonShader/Template/NM_ToonShaderParameters(BaseColor:LinearColor=<rgba 0.5804 0.5028 0.3477 1>, Lighting_Hardness=0.8, Metallic_LightingPower=0)
  ]
  emitterUpdateScript [
   &EmitterState()
   &SpawnRate(
    SpawnGroup:NiagaraInt32=&MakeIntFromFloat(Float=&FN_PickFloatDependOnPlatform(TargetPlatform=DataInterface() {propertyValues {Platforms {qualityLevelMask=-29}}}, TargetPlatformActive=1, TargetPlatformNoActive=0))
    SpawnRate=&FN_FrontEndFloatModifier(
     Gameplay=&MakeCustomFloatFromBool(
      Bool=&SetBoolByFloatComparison(A=&VectorLength(Vector=$User.OwnerLocalVelocity), B=150)
      True Float=$Emitter.Particle Spawn Rate
      False Float=$Emitter.Idle Spawn Rate
     )
     Front End=$Emitter.FE Particle Spawn Rate
    )
    Use Spawn Probability=false
   )
  ]
  particleSpawnScript [
   /SetVariables(Particles.VisibilityTag=&SpawnGroup()) {isSetParameters=true}
   &PartitionParticles()
   &InitializeParticle(
    Color=&MaskLinearColorBySpawnGroup(
     Masked Linear Color=&MultiplyLinearColor(
      Linear Color=&FNTimeOfDayColor(
       Night Color=<rgba 0.2328 0.3231 0.4936 1>
       Morning Color=<rgba 0.4416 0.4416 0.4416 1>
       Day Color=<rgba 0.6441 0.6441 0.6441 1>
       Evening Color=<rgba 0.3837 0.3837 0.3837 1>
      )
      Scale Factor=&MakeVector4_FromLinearColor(LinearColor=&MakeCustomLinearColorFromBool(
        Bool=&SetBoolByFloatComparison(A=$User.FrontEnd, B=0)
        True Color=<rgba 1 1 1 1>
        False Color=<rgba 0.7772 0.7772 0.7772 1>
       ))
     )
     Passthrough Linear Color=&MultiplyLinearColor(
      Linear Color=<rgba 0.9569 0.8275 0.5647 1>
      Scale Factor=&MakeVector4_FromLinearColor(LinearColor=&FNTimeOfDayColor(Night Color=<rgba 0.5765 0.6157 1 1>, Morning Color=<rgba 1 1 1 1>, Day Color=<rgba 1 1 1 1>, Evening Color=<rgba 1 1 1 1>))
     )
    )
    Lifetime Max=0.2
    Lifetime Min=0.3
    Lifetime Mode=::NewEnumerator1
    Mesh Scale Mode=::NewEnumerator1
    Mesh Uniform Scale Max=1
    Mesh Uniform Scale Min=1.2
    Sprite Rotation Mode=::NewEnumerator2
    Sprite Size Mode=::NewEnumerator1
    Uniform Sprite Size Max=65
    Uniform Sprite Size Min=50
    UsePositionOffset=false
    Write Material Random=true
   )
   &SocketLocation(
    "Inherit Velocity (Scale)"=0
    Orientation Sampling=::NewEnumerator1
    SamplePosOffset=true
    Sampled Position Offset=&MaskVector(Mask Source=1, Mask Parameter=$Particles.PartitionParticles.Partition, Mask=0, Masked Vector=<-5 6 -5>, Passthrough Vector=<-5 6 -4>)
    Sampled Position Offset Coordinate Space:ENiagaraSkelMeshTransforms=::NewEnumerator2
    Sampled Velocity Coordinate Space:ENiagaraCoordinateSpace=::Simulation
    Socket Index=$Particles.PartitionParticles.Partition
    Socket Reader=$System.Socket Reader
    Socket Sampling Mode=::NewEnumerator3
    VelocitySpeedLimit=false
   )
   &InitialMeshOrientation(Use Rotation Vector=false)
   &AddVelocity(Velocity=<-2 0 0.5>, Velocity Speed Scale=&RandomRangeFloat(Minimum=100, Maximum=150))
  ]
  particleUpdateScript [
   &ParticleState()
   &DynamicMaterialParameters(
    Index 0 Param 1=&FloatFromCurve(FloatCurve=DataInterface() {propertyValues {Curve {keys [
         {arriveTangent=-0.0276, time=0, value=0}
         {arriveTangent=0.0397, interpMode=RCIM_Cubic, leaveTangent=0.0397, tangentMode=RCTM_User, time=0.2445, value=0}
         {arriveTangent=2.4456, interpMode=RCIM_Cubic, leaveTangent=2.4456, tangentMode=RCTM_User, time=1, value=0.6}
        ]}}})
    Module.Index0Float0Write=true
    Module.Index0Float1Write=true
    Module.Index0Float2Write=true
    Module.Index0Float3Write=true
   ) {module=DynamicMaterialParameters001}
   &UpdateMeshOrientation(
    Rotation Rate=&FloatFromCurve(
     FloatCurve=DataInterface() {propertyValues {Curve {keys [{interpMode=RCIM_Cubic, tangentMode=RCTM_User, time=0.4, value=0}, {interpMode=RCIM_Cubic, tangentMode=RCTM_User, time=1, value=1}]}}}
     Scale Curve=&RandomRangeFloat(Minimum=-2, Maximum=2)
    )
    Rotation Vector=&RandomVector(Vector Scale=1)
   )
   &ScaleSpriteSize(Uniform Curve Sprite Scale=DataInterface() {propertyValues {Curve {keys [
        {arriveTangent=0.0398, interpMode=RCIM_Cubic, leaveTangent=0.0398, tangentMode=RCTM_User, time=0, value=0.6}
        {arriveTangent=0.3882, interpMode=RCIM_Cubic, leaveTangent=0.3882, tangentMode=RCTM_User, time=0.3227, value=0.7964}
        {arriveTangent=-0.0229, interpMode=RCIM_Cubic, leaveTangent=-0.0229, tangentMode=RCTM_User, time=1, value=1}
       ]}}})
   &Drag(
    Drag=&FloatFromCurve(
     FloatCurve=DataInterface() {propertyValues {Curve {keys [<at 0 0>,<at 0.05 1>]}}}
     Scale Curve=&RandomRangeFloat(Minimum=7, Maximum=10)
    )
    Use Linear Drag=true
    Use Rotational Drag=true
   )
   &ScaleMeshSize(Scale Factor=&VectorFromFloat(Value=&FloatFromCurve(FloatCurve=DataInterface() {propertyValues {Curve {keys [
          {arriveTangent=0.0398, interpMode=RCIM_Cubic, leaveTangent=0.0398, tangentMode=RCTM_User, time=0, value=0.3223}
          {arriveTangent=0.3486, interpMode=RCIM_Cubic, leaveTangent=0.3486, tangentMode=RCTM_User, time=0.2121, value=0.4787}
          {arriveTangent=1.9709, interpMode=RCIM_Cubic, leaveTangent=1.9709, tangentMode=RCTM_User, time=0.7675, value=1.4002}
         ]}}})))
   &ScaleColor(
    Scale RGB=&MakeVectorFromLinearColorRGB(Color=&ColorFromCurve(ColorCurve=DataInterface() {propertyValues {
        AlphaCurve {keys [<at 0 1 RCIM_Cubic>]}
        BlueCurve {keys [<at 0 1 RCIM_Cubic>,<at 1 0.5882>]}
        GreenCurve {keys [<at 0 1 RCIM_Cubic>,<at 1 0.5882>]}
        RedCurve {keys [<at 0 1 RCIM_Cubic>,<at 1 0.5882>]}
       }}))
    ScaleA=false
    ScaleRGB=true
   )
   &SolveForcesAndVelocity(Acceleration Limit=1, Clamp Velocity=true, Limit Acceleration=true, Speed Limit=1)
   &BlendSpace("Position Blend (World <-> Local)":NiagaraFloat=&MakeCustomFloatFromBool(
     Bool=&SetBoolByFloatComparison(A=&VectorLength(Vector=$Engine.Owner.Velocity), B=1000)
     True Float=1
     False Float=1
     Translation Contribution:NiagaraFloat=1
     Rotation Contribution:NiagaraFloat=0.6
    ))
   /SetVariables(Particles.Sprite Dynamic Parameters:Vector4f=&MakeVector4(
     X=&FloatFromCurve(FloatCurve=DataInterface() {propertyValues {Curve {keys [
          {arriveTangent=-0.026, interpMode=RCIM_Cubic, leaveTangent=-0.026, tangentMode=RCTM_User, time=0, value=0.2}
          {arriveTangent=2.4085, interpMode=RCIM_Cubic, leaveTangent=2.4085, tangentMode=RCTM_User, time=1, value=1}
         ]}}})
     Y=&SelectFloatFromArray(
      Array Sampling Mode=::NewEnumerator0
      Direct Array Index=?
      Float Selection Array=DataInterface() {propertyValues {FloatData [0,1]}}
      Random Seed=0
     )
     Z=0
     W=0
    )) {isSetParameters=true}
  ]
  rendererClasses [&NiagaraMeshRendererProperties, &NiagaraSpriteRendererProperties]
  renderers [
   {rendererClass=&NiagaraMeshRendererProperties, rendererIndex=0}
   {rendererClass=&NiagaraSpriteRendererProperties, rendererIndex=1}
  ]
 }}
summary {
 systemName=NS_Emote_FruitFire_Traversal
 emitters [{
   emitterName=Default_Emote_TraversalDust_Toon
   rendererClasses [&NiagaraMeshRendererProperties, &NiagaraSpriteRendererProperties]
  }]
 userVariables [
  User.bIsFrontEnd {type=NiagaraBool}
  User.bIsFrontEndPreview {type=NiagaraBool}
  User.FrontEnd {type=NiagaraFloat, defaultValue=1, description=""}
  User.OwnerLocalVelocity {type=Vector3f, description="", defaultValue {struct=Vector3f, value=<0 0 0>}}
 ]
}
